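# Niagara system for the PieMan backpack cosmetic, adding drool, flies, gloop and smoke to its skeletal mesh.
# Drool uses pendulum-constrained stretched meshes hung from bones or DroolSource particles, with location and death events.
# Flies orbit via curl noise and attraction, feeding ribbon trails through a particle reader.
# Gloop spawns from bone movement with front-end preview variants, while skydiving shuts off fly and smoke spawning.
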NiagaraSystem /BRCosmetics/Accessories/FORT_Backpacks/Backpack_M_MED_PieMan/Effects/NS_PieMan_Backpack
dependencies {
 usedDataInterfaces [
  &NiagaraDataInterfaceParticleRead
  &NiagaraDataInterfaceCurve
  /Script/Niagara.NiagaraDataInterfaceSimpleCounter
  &NiagaraDataInterfaceCamera
  &NiagaraDataInterfaceColorCurve
  &NiagaraDataInterfaceVectorCurve
 ]
 usedDynamicInputs [&MakeCustomIntFromBool]
 usedModules [
  &SystemState
  /BRCosmetics/Athena/Effects/Particles/Gliders/NS_Glider_CarbideBlack/Module/SkeletalMeshData
  &self:SystemSpawnScript.SystemScriptSource.SystemScriptGraph.NiagaraNodeAssignment_0.SetVariables_#1
 ]
}
emitters {
 Drool {
  emitterSpawnScript [/Game/Effects/Niagara/Modules/Attributes/FN_Source_v2(AttributeReader=DataInterface() {propertyValues {EmitterBinding {bindingMode=Other, emitterName=DroolSource}}})]
  emitterUpdateScript [&EmitterState()]
  particleSpawnScript [/Niagara/Modules/Spawn/Initialization/InitializeParticle(
    Color=&RandomRangeLinearColor(Minimum=<rgba 1 0.5793 0.463 1>, Maximum=<rgba 1 0.5793 0.463 1>)
    Lifetime=&UniformRangedFloat(Minimum=1, Maximum=2)
    Mesh Scale=&Multiply_VectorByFloat(Vector=<0.125 0.125 0.05>, Float=&RandomRangeFloat(Minimum=0.75, Maximum=1.5))
    Module.Write Color=true
    Module.Write Lifetime=true
    Module.Write Mass=true
    Module.Write Material Random=true
    Module.Write Position=false
    Module.Write Scale=true
    Module.Write SpriteRotation=false
    Module.Write SpriteSize=true
    Sprite Size=&Vector2DFromFloat(Value=5)
   )]
  particleUpdateScript [
   &ParticleState()
   &self:UpdateParentPosition() {module=UpdateParentPosition}
   &PendulumSetup(
    Module.Enable Spring Driver=false
    Pendulum Length:NiagaraFloat=&FloatFromCurve(
     FloatCurve=DataInterface() {propertyValues {Curve {keys [
         {arriveTangent=9.3293, interpMode=RCIM_Cubic, leaveTangent=9.3293, tangentMode=RCTM_User, time=0, value=0.1}
         {arriveTangent=-3.3621, interpMode=RCIM_Cubic, leaveTangent=-3.1784, tangentMode=RCTM_User, time=0.5, value=1}
         {arriveTangent=1.5233, interpMode=RCIM_Cubic, leaveTangent=1.5233, tangentMode=RCTM_User, time=0.8153, value=1}
         <at 1.1127 1>
        ]}}}
     Scale Curve=&RandomRangeFloat(Minimum=5, Maximum=9)
    )
    Pendulum Pivot:Vector3f=$Particles.ParentPosition
    Rigid Pendulum=false
   )
   &Drag(Drag=-0.5, Use Linear Drag=true, Use Rotational Drag=false)
   &GravityForce(Gravity=<0 0 -3000>)
   &InheritVelocity(Module.Limit Speed=true, Velocity Scale=&Lerp_Vector(A=<-1 -1 -1>, B=<0 0 0>, Alpha=$Particles.NormalizedAge))
   &SolveForcesAndVelocity(Clamp Velocity=false, Limit Acceleration=false)
   &PendulumConstraint()
   &CalculateAccurateVelocity()
   /SetVariables(Particles.MeshFacing:Vector3f=&NormalizeVector(Vector To Normalize=&Subtract_Vector(A=$Particles.ParentPosition, B=Linked() {linkedVariable {name=Particles.Position, type=NiagaraPosition}}))) {isSetParameters=true, module=SetVariables_#2}
   /Niagara/Modules/Update/Orientation/V2/OrientMeshToVector(
    Look At Direction:Vector3f=$Particles.MeshFacing
    Override Quat=false
    Reference Axis Vector:Vector3f=<0 0 1>
    UseOptionalSideAxis=false
    UseOptionalUpAxis=false
   )
   &DynamicMaterialParameters(
    Index 2 Param 1=HlslExpression() {hlslExpression=Particles.ParentPosition.x}
    Index 2 Param 2=HlslExpression() {hlslExpression=Particles.ParentPosition.y}
    Index 2 Param 3=HlslExpression() {hlslExpression=Particles.ParentPosition.z}
    Index 2 Param 4=3
    Index 3 Param 1=&MakeFloatFromVector(Vector=Linked() {linkedVariable {name=Particles.Position, type=NiagaraPosition}}, Channel=::NewEnumerator0)
    Index 3 Param 2=&MakeFloatFromVector(Vector=Linked() {linkedVariable {name=Particles.Position, type=NiagaraPosition}}, Channel=::NewEnumerator1)
    Index 3 Param 3=&MakeFloatFromVector(Vector=Linked() {linkedVariable {name=Particles.Position, type=NiagaraPosition}}, Channel=::NewEnumerator2)
    Index 3 Param 4=0.3
    Module.Param0WriteEnabled2=true
    Module.Param0WriteEnabled3=true
    Module.Param1WriteEnabled2=true
    Module.Param1WriteEnabled3=true
    Module.Param2WriteEnabled2=true
    Module.Param2WriteEnabled3=true
    Module.Param3WriteEnabled2=true
    Module.Param3WriteEnabled3=true
    Write Parameter Index 0=false
    Write Parameter Index 2=true
    Write Parameter Index 3=true
   )
   /SetVariables(Particles.XYScale:NiagaraFloat=&FloatFromCurve(FloatCurve=DataInterface() {propertyValues {Curve {keys [
         {arriveTangent=10.1742, interpMode=RCIM_Cubic, leaveTangent=10.1742, tangentMode=RCTM_User, time=0, value=0}
         {arriveTangent=-2.2106, interpMode=RCIM_Cubic, leaveTangent=-2.2106, tangentMode=RCTM_User, time=0.15, value=1}
         {arriveTangent=-0.0096, interpMode=RCIM_Cubic, leaveTangent=-0.0096, tangentMode=RCTM_User, time=0.4, value=0.5}
         {arriveTangent=-0.0242, interpMode=RCIM_Cubic, leaveTangent=-0.0242, tangentMode=RCTM_User, time=0.9, value=0.5}
         <at 1 0>
        ]}}})) {isSetParameters=true}
   &ScaleMeshSize(Scale Factor=&MakeVector(
     X=$Particles.XYScale
     Y=$Particles.XYScale
     Z=&Multiply_Float(
      A=0.2
      B=&VectorLength(Vector=&Subtract_Vector(A=$Particles.ParentPosition, B=Linked() {linkedVariable {name=Particles.Position, type=NiagaraPosition}}))
     )
    ))
   &ScaleMeshSize(Scale Factor=&MakeVector(
     X=$Particles.XYScale
     Y=$Particles.XYScale
     Z=&Multiply_Float(
      A=&FloatFromCurve(
       FloatCurve=DataInterface() {propertyValues {Curve {keys [
           <at 0 1>
           {arriveTangent=-2.1519, interpMode=RCIM_Cubic, leaveTangent=-2.1519, tangentMode=RCTM_User, time=0.4, value=1}
           {arriveTangent=-0.3667, interpMode=RCIM_Cubic, leaveTangent=-0.3667, tangentMode=RCTM_User, time=0.6719, value=0.7273}
           {arriveTangent=-0.693, interpMode=RCIM_Cubic, leaveTangent=-0.693, tangentMode=RCTM_User, time=0.9185, value=0.8098}
           <at 1 0.5>
          ]}}}
       Scale Curve=0.2
      )
      B=&VectorLength(Vector=&Subtract_Vector(A=$Particles.ParentPosition, B=Linked() {linkedVariable {name=Particles.Position, type=NiagaraPosition}}))
     )
    )) {enabled=false, module=ScaleMeshSize001}
   &KillParticlesInVolume(Origin Offset:Vector3f=<11.8793 5.8908 -10.6953>, Sphere Radius=12)
  ]
  rendererClasses [&NiagaraMeshRendererProperties, &NiagaraSpriteRendererProperties]
  renderers [
   {rendererClass=&NiagaraMeshRendererProperties, rendererIndex=0}
   {rendererClass=&NiagaraSpriteRendererProperties, rendererIndex=1}
  ]
 }
 Drool001 {
  emitterUpdateScript [&EmitterState()]
  particleSpawnScript [
   &InitializeParticle(
    Lifetime=&FN_FrontEndFloatModifier(
     Gameplay=&Lerp_Float(A=1, B=0, Alpha=&ClampFloat(Float=&Multiply_Float(A=2, B=$User.ForwardDot)))
     Front End=&RandomRangeFloat(Minimum=0.25, Maximum=0.5)
    )
    Mesh Scale=&Lerp_Vector(A=<3 1 1>, B=<10 4 4>, Alpha=$Particles.RandomNormalizedFloat)
    Module.Write Color=true
    Module.Write Lifetime=true
    Module.Write Mass=false
    Module.Write Material Random=true
    Module.Write Position=false
    Module.Write Scale=false
    Module.Write SpriteRotation=false
    Module.Write SpriteSize=true
    Sprite Size=&Vector2DFromFloat(Value=5)
   )
   &AddVelocity(Coordinate Space=::World, Velocity=<0 0 -200>)
  ]
  particleUpdateScript [
   &ParticleState()
   &DynamicMaterialParameters(
    Index 0 Param 1=&RandomRangeFloat(Minimum=0, Maximum=1)
    Index 0 Param 2=1
    Module.Param0WriteEnabled=true
    Module.Param1WriteEnabled=true
    Module.Param2WriteEnabled=true
    Module.Param3WriteEnabled=true
   )
   &ScaleMeshSize(Scale Factor=&Multiply_VectorByFloat(
     Vector=&MakeVector(
      X=&Lerp_Float(A=0.25, B=1, Alpha=$Particles.NormalizedAge)
      Y=&Lerp_Float(A=1, B=0.5, Alpha=$Particles.NormalizedAge)
      Z=&Lerp_Float(A=1, B=0.5, Alpha=$Particles.NormalizedAge)
     )
     Float=&Lerp_Float(A=1, B=0, Alpha=$Particles.NormalizedAge)
    )) {enabled=false}
   &InheritVelocity(Module.Limit Speed=true, Velocity Scale=<-0.1 -0.1 -0.1>)
   &GravityForce()
   &SolveForcesAndVelocity(Clamp Velocity=false, Limit Acceleration=false)
  ]
  rendererClasses [&NiagaraSpriteRendererProperties, &NiagaraMeshRendererProperties]
  renderers [
   {rendererClass=&NiagaraSpriteRendererProperties}
   {rendererClass=&NiagaraMeshRendererProperties, rendererIndex=1}
  ]
 }
 DroolPendulum {
  emitterUpdateScript [
   &EmitterState()
   &SpawnRate(Spawn Probability=0.1, SpawnRate=3, Use Spawn Probability=true)
  ]
  particleSpawnScript [
   &SkeletalMeshLocation(Bone / Socket Index=3, Bone Sampling Mode=::NewEnumerator4, Inherit Velocity=false, MaskSpawnGroup=false, Skeletal Mesh=$User.SkeletalMesh) {module=SkeletalMeshLocation002}
   /SetVariables(
    Particles.PendulumLength:NiagaraFloat=&RandomRangeFloat(Minimum=15, Maximum=25)
    Particles.ScaleLerp:NiagaraFloat=&RandomRangeFloat(Minimum=1, Maximum=1)
   ) {isSetParameters=true}
   &InitializeParticle(
    Lifetime=&Lerp_Float(A=0.8, B=2, Alpha=$Particles.ScaleLerp)
    Mesh Scale=&Lerp_Vector(A=<0.1 0.1 0.0175>, B=<0.1875 0.1875 0.0175>, Alpha=$Particles.ScaleLerp)
    Module.Write Color=true
    Module.Write Lifetime=true
    Module.Write Mass=false
    Module.Write Material Random=true
    Module.Write Position=false
    Module.Write Scale=true
    Module.Write SpriteRotation=false
    Module.Write SpriteSize=true
    Sprite Size=&Vector2DFromFloat(Value=5)
   )
  ]
  particleUpdateScript [
   &ParticleState()
   &SkeletalMeshLocation(
    Apply To Particle Position=false
    Bone / Socket Index=3
    Bone Sampling Mode=::NewEnumerator4
    Inherit Velocity=false
    MaskSpawnGroup=false
    Skeletal Mesh=$User.SkeletalMesh
   )
   /SetVariables(Particles.MeshPosition:Vector3f=&Add_Vector(
     A=$Output.SkeletalMeshLocation.SampledPosition
     B=&RotateVectorByQuaternion(
      Vector To Rotate=&Add_Vector(
       A=<-1 0 0>
       B=&Multiply_Vector(A=<0.5 0.5 1.5>, B=&RandomVector(Vector Scale=0.5))
      )
      Quaternion=$Output.SkeletalMeshLocation.SampledRotation
     )
    )) {isSetParameters=true}
   /SetVariables(
    Particles.AlphaScale:NiagaraFloat=&FloatFromCurve(FloatCurve=DataInterface() {propertyValues {Curve {keys [
         {arriveTangent=-0.0275, interpMode=RCIM_Cubic, leaveTangent=-0.0275, tangentMode=RCTM_User, time=0.95, value=1}
         {arriveTangent=-34.2468, interpMode=RCIM_Cubic, leaveTangent=-34.2466, tangentMode=RCTM_User, time=1, value=0}
        ]}}})
    Particles.XYScale:NiagaraFloat=&FloatFromCurve(FloatCurve=DataInterface() {propertyValues {Curve {keys [
         {arriveTangent=10.1742, interpMode=RCIM_Cubic, leaveTangent=10.1742, tangentMode=RCTM_User, time=0, value=0}
         {arriveTangent=-2.2106, interpMode=RCIM_Cubic, leaveTangent=-2.2106, tangentMode=RCTM_User, time=0.15, value=1}
         {arriveTangent=-0.0096, interpMode=RCIM_Cubic, leaveTangent=-0.0096, tangentMode=RCTM_User, time=0.4, value=0.5}
         {arriveTangent=-0.0242, interpMode=RCIM_Cubic, leaveTangent=-0.0242, tangentMode=RCTM_User, time=0.9, value=0.5}
         <at 1 0.2>
        ]}}})
    Particles.ZScale:NiagaraFloat=&FloatFromCurve(
     FloatCurve=DataInterface() {propertyValues {Curve {keys [
         {arriveTangent=3.8833, interpMode=RCIM_Cubic, leaveTangent=3.8833, tangentMode=RCTM_User, time=-0.0037, value=0.05}
         {arriveTangent=0.939, interpMode=RCIM_Cubic, leaveTangent=0.939, tangentMode=RCTM_User, time=0.872, value=0.3942}
         {arriveTangent=4.1686, interpMode=RCIM_Cubic, leaveTangent=4.1686, tangentMode=RCTM_User, time=1.1127, value=1}
        ]}}}
     Scale Curve=$Particles.PendulumLength
    )
   ) {isSetParameters=true}
   &PendulumSetup(
    Module.Enable Spring Driver=false
    Pendulum Length:NiagaraFloat=$Particles.ZScale
    Pendulum Pivot:Vector3f=$Particles.MeshPosition
    Rigid Pendulum=false
   )
   &Drag(Drag=2, Use Linear Drag=true, Use Rotational Drag=false)
   &GravityForce()
   &SolveForcesAndVelocity(Clamp Velocity=false, Limit Acceleration=false)
   &PendulumConstraint()
   &CalculateAccurateVelocity()
   /SetVariables(Particles.MeshFacing:Vector3f=&NormalizeVector(Vector To Normalize=&Subtract_Vector(A=$Particles.MeshPosition, B=Linked() {linkedVariable {name=Particles.Position, type=NiagaraPosition}}))) {isSetParameters=true, module=SetVariables_#2}
   &OrientMeshToVector(
    Look At Direction:Vector3f=&Multiply_VectorByFloat(Vector=$Particles.MeshFacing, Float=1)
    Override Quat=false
    Reference Axis Vector:Vector3f=<0 0 1>
    UseOptionalSideAxis=false
    UseOptionalUpAxis=false
   )
   /SetVariables(Particles.RestDirection:Vector3f=&TransformVector(Vector=<0 0 -1>, Source Space=::Simulation, Destination Space=::World)) {isSetParameters=true}
   &DynamicMaterialParameters(
    Index 1 Param 1=HlslExpression() {hlslExpression=Particles.RestDirection.x}
    Index 1 Param 2=HlslExpression() {hlslExpression=Particles.RestDirection.y}
    Index 1 Param 3=HlslExpression() {hlslExpression=Particles.RestDirection.z}
    Index 2 Param 1=HlslExpression() {hlslExpression=Particles.MeshPosition.x}
    Index 2 Param 2=HlslExpression() {hlslExpression=Particles.MeshPosition.y}
    Index 2 Param 3=HlslExpression() {hlslExpression=Particles.MeshPosition.z}
    Index 2 Param 4=0.75
    Index 3 Param 1=&MakeFloatFromVector(Vector=Linked() {linkedVariable {name=Particles.Position, type=NiagaraPosition}}, Channel=::NewEnumerator0)
    Index 3 Param 2=&MakeFloatFromVector(Vector=Linked() {linkedVariable {name=Particles.Position, type=NiagaraPosition}}, Channel=::NewEnumerator1)
    Index 3 Param 3=&MakeFloatFromVector(Vector=Linked() {linkedVariable {name=Particles.Position, type=NiagaraPosition}}, Channel=::NewEnumerator2)
    Index 3 Param 4=0.3
    Module.Param0WriteEnabled1=true
    Module.Param0WriteEnabled2=true
    Module.Param0WriteEnabled3=true
    Module.Param1WriteEnabled1=true
    Module.Param1WriteEnabled2=true
    Module.Param1WriteEnabled3=true
    Module.Param2WriteEnabled1=true
    Module.Param2WriteEnabled2=true
    Module.Param2WriteEnabled3=true
    Module.Param3WriteEnabled1=false
    Module.Param3WriteEnabled2=true
    Module.Param3WriteEnabled3=true
    Write Parameter Index 0=false
    Write Parameter Index 1=true
    Write Parameter Index 2=true
    Write Parameter Index 3=true
   )
   &ScaleMeshSize(Scale Factor=&MakeVector(
     X=$Particles.XYScale
     Y=$Particles.XYScale
     Z=&VectorLength(Vector=&Subtract_Vector(A=$Particles.MeshPosition, B=Linked() {linkedVariable {name=Particles.Position, type=NiagaraPosition}}))
    ))
   &Color(Color=&ColorFromCurve(
     ColorCurve=DataInterface() {propertyValues {
       AlphaCurve {keys [<at 0 20 RCIM_Cubic>]}
       BlueCurve {keys [<at 0 0 RCIM_Cubic>,<at 0.5257 0>,<at 1 5 RCIM_Cubic>]}
       GreenCurve {keys [<at 0 0 RCIM_Cubic>,<at 0.5257 5>,<at 1 0 RCIM_Cubic>]}
       RedCurve {keys [<at 0 5 RCIM_Cubic>,<at 0.5257 0>,<at 1 0 RCIM_Cubic>]}
      }}
     CurveIndex=&Divide_Float(A=&MakeFloatFromInt(Int=$Particles.PickBone), B=366)
    )) {enabled=false}
   &ScaleColor(Scale Alpha=$Particles.AlphaScale, ScaleA=true, ScaleRGB=false)
   &GenerateDeathEvent(Float to Send as Random Float:NiagaraFloat=$Particles.ScaleLerp)
  ]
  rendererClasses [&NiagaraSpriteRendererProperties, &NiagaraMeshRendererProperties]
  renderers [
   {rendererClass=&NiagaraSpriteRendererProperties}
   {rendererClass=&NiagaraMeshRendererProperties, rendererIndex=1}
   {rendererClass=&NiagaraSpriteRendererProperties, rendererIndex=2}
  ]
 }
 DroolSource {
  emitterUpdateScript [
   &EmitterState()
   &SpawnRate(Spawn Probability=0.5, SpawnRate=5, Use Spawn Probability=true)
  ]
  particleSpawnScript [
   &InitializeParticle(
    Lifetime=2
    Module.Write Color=true
    Module.Write Lifetime=true
    Module.Write Mass=false
    Module.Write Material Random=true
    Module.Write Position=false
    Module.Write Scale=false
    Module.Write SpriteRotation=false
    Module.Write SpriteSize=true
    Sprite Size=&Vector2DFromFloat(Value=10)
   )
   /SetVariables(Particles.HangTime=0.7, Particles.PickBone:NiagaraInt32=&RandomRangeInt(Minimum=1, Maximum=2)) {isSetParameters=true}
   &SkeletalMeshLocation(Bone / Socket Index=1, Bone Sampling Mode=::NewEnumerator4, Inherit Velocity=false, MaskSpawnGroup=false, Skeletal Mesh=$User.SkeletalMesh) {module=SkeletalMeshLocation001}
   &ApplyInitialForces()
   &GenerateLocationEvent(Module.Use Event Send Rate=true)
  ]
  particleUpdateScript [
   &ParticleState()
   &SkeletalMeshLocation(
    Apply To Particle Position:NiagaraBool=&SetBoolByFloatComparison(A=$Particles.NormalizedAge, B=$Particles.HangTime, Comparison Type=::NewEnumerator6)
    Bone / Socket Index=$Particles.PickBone
    Bone Sampling Mode=::NewEnumerator4
    Inherit Velocity=false
    MaskSpawnGroup=false
    Sampled Position Offset=&Lerp_Vector(
     A=&RandomRangeVector(Minimum=<0 0 -1>, Maximum=<20 2 1>)
     B=&RandomRangeVector(Minimum=<5 0 0>, Maximum=<13 2 -1.5>)
     Alpha=&MakeFloatFromInt(Int=&Add_Integer(A=$Particles.PickBone, B=-1))
    )
    Sampled Position Offset Coordinate Space:ENiagaraMeshTransforms=::NewEnumerator3
    Skeletal Mesh=$User.SkeletalMesh
   )
   &InheritVelocity(
    Module.Limit Speed=true
    Velocity Scale=&Lerp_Vector(A=<-0.01 -0.01 -0.01>, B=<0 0 0>, Alpha=$Particles.NormalizedAge)
   ) {enabled=false}
   &GravityForce(Gravity=&Lerp_Vector(A=<0 0 150>, B=<0 0 -300>, Alpha=$Particles.NormalizedAge))
   &SolveForcesAndVelocity(Clamp Velocity=false, Limit Acceleration=false)
  ]
  rendererClasses [&NiagaraSpriteRendererProperties]
  renderers [{rendererClass=&NiagaraSpriteRendererProperties}]
 }
 Flies_Optimized {
  emitterUpdateScript [
   &EmitterState()
   &SpawnRate(
    Spawn Probability=0.25
    SpawnRate=&Lerp_Float(A=4, B=0, Alpha=&MakeFloatFromInt(Int=$System.isSkydiving))
    Use Spawn Probability=true
   )
  ]
  particleSpawnScript [
   &InitializeParticle(
    Color=&RandomRangeLinearColor(Minimum=<rgba 0 0.0057 0.057 1>, Maximum=<rgba 0.005 0.02 0 1>)
    Lifetime=&Lerp_Float(A=1, B=3, Alpha=&Power(a=&RandomRangeFloat(Minimum=0, Maximum=1), b=4))
    Mesh Scale=&Multiply_VectorByFloat(Vector=<1 2 1>, Float=&UniformRangedFloat(Minimum=0.1, Maximum=0.15))
    Module.Write Color=true
    Module.Write Lifetime=true
    Module.Write Mass=false
    Module.Write Material Random=true
    Module.Write Position=false
    Module.Write Scale=true
    Module.Write SpriteRotation=false
    Module.Write SpriteSize=false
    Position=&SimulationPosition()
   )
   &DynamicMaterialParameters(
    Index 0 Param 1=10
    Index 0 Param 2=1
    Index 1 Param 1=&UniformRangedFloat(Minimum=0, Maximum=30)
    Index 1 Param 2=&UniformRangedFloat(Minimum=0, Maximum=1)
    Index 1 Param 3=&UniformRangedFloat(Minimum=0, Maximum=1)
    Index 1 Param 4=&UniformRangedFloat(Minimum=0, Maximum=1)
    Module.Param0WriteEnabled=true
    Module.Param0WriteEnabled1=true
    Module.Param1WriteEnabled=true
    Module.Param1WriteEnabled1=false
    Module.Param2WriteEnabled=true
    Module.Param2WriteEnabled1=false
    Module.Param3WriteEnabled=true
    Module.Param3WriteEnabled1=false
    Write Parameter Index 1=true
   )
   &SkeletalMeshLocation(
    Bone Sampling Mode=::NewEnumerator4
    Inherit Velocity=false
    MaskSpawnGroup=false
    Sampled Position Offset=&RandomRangeVector(Minimum=<-2 -13 -4>, Maximum=<5 -6 -4>)
    Sampled Position Offset Coordinate Space:ENiagaraMeshTransforms=::NewEnumerator3
    Skeletal Mesh=$User.SkeletalMesh
   )
   &AddVelocity(Velocity=&Multiply_Vector(
     A=&Subtract_Vector(A=$Output.SkeletalMeshLocation.SampledPosition, B=Linked() {linkedVariable {name=Particles.Position, type=NiagaraPosition}})
     B=<2 2 0.25>
    ))
  ]
  particleUpdateScript [
   &ParticleState()
   &SkeletalMeshLocation(
    Apply To Particle Position=false
    Bone / Socket Index=2
    Bone Sampling Mode=::NewEnumerator4
    Inherit Velocity=false
    MaskSpawnGroup=false
    Sampled Position Offset=&Add_Vector(
     A=<-22.1231 -11 -0.8408>
     B=&RotateVectorByQuaternion(Vector To Rotate=&RandomRangeVector(Minimum=<0 16.1011 0>, Maximum=<30 0 2>), Quaternion=<0 -0.15 0 1>)
    )
    Sampled Position Offset Coordinate Space:ENiagaraMeshTransforms=::NewEnumerator3
    Skeletal Mesh=$User.SkeletalMesh
   ) {module=SkeletalMeshLocation001}
   /SetVariables(Particles.SourceDistance:NiagaraFloat=/Niagara/DynamicInputs/TypeConversions/VectorLengthSafe(Vector:Vector3f=&Subtract_Vector(A=$Output.SkeletalMeshLocation001.SampledPosition, B=Linked() {linkedVariable {name=Particles.Position, type=NiagaraPosition}}))) {isSetParameters=true}
   /Niagara/Modules/Update/Forces/CurlNoiseForce(
    Noise Frequency:NiagaraFloat=&RandomRangeFloat(Minimum=0.05, Maximum=0.25)
    Noise Strength=1000
    Pan Noise=true
    Pan Noise Field=<0 0 1>
   )
   /Niagara/Modules/Update/Forces/PointAttractionForce(
    Attraction Radius=1000
    AttractionStrength:NiagaraFloat=&Lerp_Float(A=0, B=1, Alpha=$Particles.SourceDistance)
    AttractorPosition:Vector3f=$Output.SkeletalMeshLocation001.SampledPosition
    Module.Kill Within Radius=false
    Module.Use Falloff=true
   ) {module=PointAttractionForce002}
   &Drag(Use Linear Drag=true, Use Rotational Drag=false)
   &DynamicMaterialParameters(
    Index 0 Param 1=&UniformRangedFloat(Minimum=4, Maximum=5)
    Index 0 Param 2=&FloatFromCurve(FloatCurve=DataInterface() {propertyValues {Curve {keys [
         {interpMode=RCIM_Cubic, tangentMode=RCTM_User, time=0, value=0}
         {arriveTangent=0.002, interpMode=RCIM_Cubic, leaveTangent=0.002, tangentMode=RCTM_User, time=0.1827, value=1}
         {arriveTangent=-1.2195, interpMode=RCIM_Cubic, leaveTangent=-1.2195, tangentMode=RCTM_User, time=0.9833, value=0.6448}
         <at 1.0011 0.002>
        ]}}})
    Index 1 Param 1=&UniformRangedFloat(Minimum=0, Maximum=30)
    Index 1 Param 2=1.1418
    Index 1 Param 3=&UniformRangedFloat(Minimum=0, Maximum=1)
    Index 1 Param 4=&UniformRangedFloat(Minimum=0, Maximum=1)
    Module.Param0WriteEnabled=false
    Module.Param0WriteEnabled1=true
    Module.Param1WriteEnabled=true
    Module.Param1WriteEnabled1=false
    Module.Param2WriteEnabled=false
    Module.Param2WriteEnabled1=true
    Module.Param3WriteEnabled=false
    Module.Param3WriteEnabled1=true
    Write Parameter Index 1=true
   ) {module=DynamicMaterialParameters001}
   &SolveForcesAndVelocity(Clamp Velocity=false, Limit Acceleration=false)
   &ScaleMeshSize(Scale Factor=&VectorFromFloat(Value=&FloatFromCurve(
      FloatCurve=DataInterface() {propertyValues {Curve {keys [
          {interpMode=RCIM_Cubic, tangentMode=RCTM_User, time=-0.0046, value=1.0113}
          {arriveTangent=-1.0067, interpMode=RCIM_Cubic, leaveTangent=-1.0067, tangentMode=RCTM_User, time=0.9283, value=0.7392}
          <at 1 0>
         ]}}}
      Scale Curve=2
     )))
  ]
  rendererClasses [&NiagaraMeshRendererProperties]
  renderers [{rendererClass=&NiagaraMeshRendererProperties, rendererIndex=0}]
 }
 Fly_Trail {
  emitterUpdateScript [
   &EmitterState(Enable Distance Culling=true, MaxDistance=2500, Module.bUseMaxDistance=true, Module.bUseMinDistance=false, Scalability Mode=::NewEnumerator1)
   /SetVariables(Emitter.EmitterReader=DataInterface() {propertyValues {EmitterBinding {bindingMode=Other, emitterName=Flies_Optimized}}}) {isSetParameters=true}
   &SpawnRate(
    SpawnRate=/Game/Effects/Niagara/DynamicInputs/Attributes/FN_GetEmitterParticleCount(AttributeReader:NiagaraDataInterfaceParticleRead=$Emitter.EmitterReader, Multiplier:NiagaraFloat=20)
    Use Spawn Probability=true
   )
  ]
  particleSpawnScript [
   &self:Sample_Location(New Particle Attribute Reader:NiagaraDataInterfaceParticleRead=$Emitter.EmitterReader) {module=Sample_Location}
   &InitializeRibbon(
    Color=<rgba 1 1 1 0.4>
    Lifetime=0.5
    Module.Write Color=true
    Module.Write Lifetime=true
    Module.Write Mass=false
    Module.Write Material Random=true
    Module.Write Position=false
    Module.Write RibbonTwist=false
    Module.Write RibbonWidth=true
    Ribbon Width=&RandomRangeFloat(Minimum=2, Maximum=6)
   )
  ]
  particleUpdateScript [
   &ParticleState()
   /SetVariables(Particles.DistanceFromBackpack:NiagaraFloat=&Multiply_Float(
     A=&Add_Float(
      A=&VectorLength(Vector=&Subtract_Vector(
        A=Linked() {linkedVariable {name=Particles.Position, type=NiagaraPosition}}
        B=Linked() {linkedVariable {name=Engine.Owner.Position, type=NiagaraPosition}}
       ))
      B=-50
     )
     B=0.02
    )) {isSetParameters=true}
   &SolveForcesAndVelocity(Clamp Velocity=false, Limit Acceleration=false)
   &ScaleColor(
    Scale Alpha=/Niagara/DynamicInputs/Math/LerpMultipleFloats(
     Float Count:ENiagara_MultipleLerpCount=::NewEnumerator2
     1:NiagaraFloat=0
     2:NiagaraFloat=1
     3:NiagaraFloat=1
     4:NiagaraFloat=0
     5:NiagaraFloat=0
     Alpha:NiagaraFloat=$Particles.NormalizedAge
    )
    ScaleA=true
    ScaleRGB=false
   )
   &ScaleColor(
    Color Value To Scale=$Particles.Color
    Scale Alpha=&Lerp_Float(A=1, B=0, Alpha=$Particles.DistanceFromBackpack)
    ScaleA=true
    ScaleRGB=false
   ) {module=ScaleColor001}
   &ScaleRibbonWidth(Ribbon Width Scale=&FloatFromCurve(
     FloatCurve=DataInterface() {propertyValues {Curve {keys [<at 0 1 RCIM_Cubic>,<at 1 0>]}}}
     Scale Curve=&VectorLength(Vector=$Particles.InheritedScale)
    ))
   &Color(Color=&ColorFromCurve(
     ColorCurve=DataInterface() {propertyValues {
       AlphaCurve {keys [<at 0 1 RCIM_Cubic>]}
       BlueCurve {keys [<at 0 0 RCIM_Cubic>,<at 0.3333 0>,<at 0.6667 0>,<at 1 1 RCIM_Cubic>]}
       GreenCurve {keys [<at 0 0 RCIM_Cubic>,<at 0.3333 1>,<at 0.6667 1>,<at 1 0 RCIM_Cubic>]}
       RedCurve {keys [<at 0 1 RCIM_Cubic>,<at 0.3333 1>,<at 0.6667 0>,<at 1 0 RCIM_Cubic>]}
      }}
     CurveIndex=$Particles.DistanceFromBackpack
    )) {enabled=false}
  ]
  rendererClasses [&NiagaraRibbonRendererProperties]
  renderers [{rendererClass=&NiagaraRibbonRendererProperties}]
 }
 Main_Gloop {
  emitterUpdateScript [
   &EmitterState(Enable Distance Culling=true, MaxDistance=7000, Module.bUseMaxDistance=true, Module.bUseMinDistance=false, Scalability Mode=::NewEnumerator1)
   &SpawnPerUnit(
    Module.Use Max Movement Threshold=true
    Module.Use Movement Tolerance=false
    Spawn Per Unit=2
    Use Spawn Probability=false
    Velocity Vector:Vector3f=&self:GetBoneVelocity(Filtered Bone Index:NiagaraInt32=4, Skeletal Mesh:NiagaraDataInterfaceSkeletalMesh=$User.SkeletalMesh)
   )
   &SpawnRate(SpawnRate=20, Use Spawn Probability=false) {enabled=false}
  ]
  particleSpawnScript [
   &SkeletalMeshLocation(
    Bone / Socket Index=4
    Bone Sampling Mode=::NewEnumerator4
    Inherit Velocity=false
    MaskSpawnGroup=false
    Sampled Position Offset=&FN_FrontEndVector3Modifier(Gameplay=<-2 0 -5>, Front End=<-2 0 -4>)
    Sampled Position Offset Coordinate Space:ENiagaraMeshTransforms=::NewEnumerator2
    Skeletal Mesh=$User.SkeletalMesh
   ) {module=SkeletalMeshLocation001}
   &SkeletalMeshLocation(
    Apply To Particle Position=false
    Bone / Socket Index=5
    Bone Sampling Mode=::NewEnumerator4
    Inherit Velocity=false
    MaskSpawnGroup=false
    Sampled Position Offset=&FN_FrontEndVector3Modifier(Gameplay=<-2 0 -5>, Front End=<-2 0 -4>)
    Sampled Position Offset Coordinate Space:ENiagaraMeshTransforms=::NewEnumerator2
    Skeletal Mesh=$User.SkeletalMesh
   ) {module=SkeletalMeshLocation002}
   &SkeletalMeshLocation(
    Apply To Particle Position=false
    Bone Sampling Mode=::NewEnumerator4
    Inherit Velocity=false
    MaskSpawnGroup=false
    Sampled Position Offset=&FN_FrontEndVector3Modifier(Gameplay=<-2 0 -5>, Front End=<-2 0 -4>)
    Sampled Position Offset Coordinate Space:ENiagaraMeshTransforms=::NewEnumerator2
    Skeletal Mesh=$User.SkeletalMesh
   ) {module=SkeletalMeshLocation003}
   &InitializeParticle(
    Color=&MakeCustomLinearColorFromBool(
     Bool=$User.bIsFrontEndPreview
     True Color=<rgba 0.577 0.5071 0.4022 1.5>
     False Color=&RandomRangeLinearColor(Minimum=<rgba 0.1096 0.1214 0.188 1.5>, Maximum=<rgba 0.1973 0.2131 0.303 1.5>)
    )
    Lifetime=&Lerp_Float(
     A=0.4
     B=&RandomRangeFloat(Minimum=0.3, Maximum=0.35)
     Alpha=&ClampFloat(Float=&Add_Float(
       A=&Multiply_Float(A=0.02, B=&VectorLength(Vector=$Output.SkeletalMeshLocation001.SampledVelocity))
       B=-0.2
      ))
    )
    Module.Write Color=true
    Module.Write Lifetime=true
    Module.Write Mass=false
    Module.Write Material Random=true
    Module.Write Position=false
    Module.Write Scale=false
    Module.Write SpriteRotation=true
    Module.Write SpriteSize=true
    Position=&SimulationPosition()
    Sprite Rotation=&UniformRangedFloat(Minimum=0, Maximum=360)
    Sprite Size=&Vector2DFromFloat(Value=&UniformRangedFloat(Minimum=4.5, Maximum=&RandomRangeFloat(Minimum=9, Maximum=11.25)))
   )
   /Niagara/Modules/Update/SubUV/SubUVAnimation(Number Of Frames=3, SubUV Animation Mode=::NewEnumerator1)
   &AddVelocityFromPoint(
    Module.Use Falloff Distance=false
    Origin Offset=<-1 -11 1>
    Velocity Origin:Vector3f=$Output.SkeletalMeshLocation002.SampledPosition
    Velocity Strength=1
   )
   &AddVelocity(Coordinate Space=::Simulation, Velocity=<0 0 -0.1>)
   /SetVariables(
    Particles.PositionTest:Vector3f=&Add_Vector(A=<2 -12 1>, B=$Output.SkeletalMeshLocation002.SampledPosition)
    Particles.ScaleTest:Vector2f=<5 5>
   ) {isSetParameters=true}
  ]
  particleUpdateScript [
   &self:SetSpriteFacing(New Camera Query=DataInterface() {propertyValues {PlayerControllerIndex=0, bRequireCurrentFrameData=false}}) {module=SetSpriteFacing}
   /SetVariables(Particles.SpriteFacing=&Lerp_Vector(
     A=/Niagara/DynamicInputs/Helpers/SafeNormalizeVector(Vector To Normalize:Vector3f=$Particles.Velocity, Fallback Vector:Vector3f=<0 0 0>)
     B=$Particles.SpriteFacingCamera
     Alpha=&FN_FrontEndFloatModifier(Gameplay=0.8, Front End=0.5)
    )) {isSetParameters=true}
   &ParticleState()
   &ScaleColor(
    Scale RGB=&VectorFromCurve(
     VectorCurve=DataInterface() {propertyValues {XCurve {keys [<at 0 1.5>,<at 1 1>]}, YCurve {keys [<at 0 1.5>,<at 1 1>]}, ZCurve {keys [<at 0 1.5>,<at 1 1>]}}}
     Scale Curve=&VectorFromFloat(Value=&LerpMultipleFloats(
       Float Count:ENiagara_MultipleLerpCount=::NewEnumerator0
       1:NiagaraFloat=0.3
       2:NiagaraFloat=1
       3:NiagaraFloat=0.3
       4:NiagaraFloat=?
       5:NiagaraFloat=?
       Alpha:NiagaraFloat=Linked() {linkedVariable {name=NPC.FortniteNPC.TimeOfDay_0-1, type=NiagaraFloat}}
      ))
    )
    ScaleA=false
    ScaleRGB=true
   )
   &ScaleSpriteSize(Scale Factor=&Vector2DFromFloat(Value=&FloatFromCurve(FloatCurve=DataInterface() {propertyValues {Curve {keys [
          {arriveTangent=7.4318, interpMode=RCIM_Cubic, leaveTangent=7.4318, tangentMode=RCTM_User, time=0, value=0}
          {arriveTangent=3.334, interpMode=RCIM_Cubic, leaveTangent=3.334, tangentMode=RCTM_User, time=0.263, value=1.2745}
          <at 2 3 RCIM_Cubic>
         ]}}})))
   &DynamicMaterialParameters(
    Index 0 Param 1=&FloatFromCurve(FloatCurve=DataInterface() {propertyValues {Curve {keys [
         {interpMode=RCIM_Cubic, tangentMode=RCTM_Break, time=0, value=1}
         {arriveTangent=-1.1592, interpMode=RCIM_Cubic, leaveTangent=-1.1592, tangentMode=RCTM_User, time=1, value=0}
        ]}}})
    Index 0 Param 2=&FloatFromCurve(FloatCurve=DataInterface() {propertyValues {Curve {keys [{arriveTangent=0.4361, interpMode=RCIM_Cubic, leaveTangent=0.4361, tangentMode=RCTM_User, time=0, value=0}, <at 1 0.125>]}}})
    Module.Param0WriteEnabled=true
    Module.Param1WriteEnabled=true
    Module.Param2WriteEnabled=false
    Module.Param3WriteEnabled=false
   )
   &SpriteRotationRate(Rotation Rate=&Multiply_Float(
     A=&Lerp_Float(A=-1, B=1, Alpha=$Particles.Initial.MaterialRandom)
     B=&Lerp_Float(A=125, B=0, Alpha=&Power(a=$Particles.NormalizedAge, b=0.2))
    ))
   &SkeletalMeshLocation(
    Apply To Particle Position=false
    Bone / Socket Index=1
    Bone Sampling Mode=::NewEnumerator4
    Inherit Velocity=false
    MaskSpawnGroup=false
    Sampled Position Offset=<-2 0 -5>
    Sampled Position Offset Coordinate Space:ENiagaraMeshTransforms=::NewEnumerator2
    Skeletal Mesh=$User.SkeletalMesh
   )
   &KillParticles(Kill Particles=&Boolean_AndOperation(
     Bool1=&Boolean_NotOperation(Bool=$User.bIsFrontEndPreview)
     Bool2=&SetBoolByFloatComparison(
      A=&MakeFloatFromVector(
       Vector=&TransformVector(Vector=Linked() {linkedVariable {name=Particles.Position, type=NiagaraPosition}}, Source Space=::Simulation, Destination Space=::World)
       Channel=::NewEnumerator2
      )
      B=&Add_Float(
       A=-1
       B=&MakeFloatFromVector(Vector=/Niagara/DynamicInputs/Helpers/SimulationPosition(), Channel=::NewEnumerator2)
      )
     )
    )) {module=KillParticles001}
   &KillParticlesInVolume(
    Box Size:Vector3f=<100 100 53>
    Kill Shape=::NewEnumerator1
    Kill Volume Enabled:NiagaraBool=$User.bIsFrontEndPreview
    Volume Origin:Vector3f=$Output.SkeletalMeshLocation.SampledPosition
   )
   &Color(Color=&ColorFromCurve(
     ColorCurve=DataInterface() {propertyValues {
       AlphaCurve {keys [<at 0 0.2 RCIM_Cubic>]}
       BlueCurve {keys [<at 0 0 RCIM_Cubic>,<at 0.3333 0>,<at 0.6667 0>,<at 1 1 RCIM_Cubic>]}
       GreenCurve {keys [<at 0 0 RCIM_Cubic>,<at 0.3333 1>,<at 0.6667 1>,<at 1 0 RCIM_Cubic>]}
       RedCurve {keys [<at 0 1 RCIM_Cubic>,<at 0.3333 1>,<at 0.6667 0>,<at 1 0 RCIM_Cubic>]}
      }}
     CurveIndex=$Particles.DistanceFromKillSphere
    )) {enabled=false}
   &CameraOffset(Camera Offset Amount=5) {enabled=false}
   &SolveForcesAndVelocity(Clamp Velocity=false, Limit Acceleration=false)
  ]
  rendererClasses [&NiagaraSpriteRendererProperties]
  renderers [
   {rendererClass=&NiagaraSpriteRendererProperties}
   {rendererClass=&NiagaraSpriteRendererProperties, rendererIndex=1}
  ]
 }
 NE_Empty {
  bEnabled=false
  emitterUpdateScript [
   &EmitterState()
   &SpawnRate(SpawnRate=400, Use Spawn Probability=false)
   /Niagara/Modules/Spawn/Location/SpawnParticlesInGrid(X Count=20, Y Count=20, Z Count=20) {enabled=false}
  ]
  particleSpawnScript [
   &InitializeParticle(
    Module.Write Color=true
    Module.Write Lifetime=true
    Module.Write Mass=true
    Module.Write Material Random=true
    Module.Write Position=false
    Module.Write Scale=false
    Module.Write SpriteRotation=false
    Module.Write SpriteSize=true
    Position=/Niagara/Functions/Localspace/SimulationPosition()
    Sprite Size=<2 2>
   )
   /Niagara/Modules/Spawn/Location/V2/GridLocation(
    Dimensions Definition:ENiagaraGridPlacementType=::NewEnumerator1
    Grid Origin:NiagaraPosition=/Niagara/DynamicInputs/Helpers/SimulationPosition()
    Module.Jitter Inside Grid Cells=false
    Module.Override Random Seed=false
    XYZ Dimensions:Vector3f=<400 400 100>
   ) {enabled=false}
   &SkeletalMeshLocation(
    Bone / Socket Index=1
    Bone Sampling Mode=::NewEnumerator4
    Inherit Velocity=false
    MaskSpawnGroup=false
    Sampled Position Offset Coordinate Space:ENiagaraMeshTransforms=::NewEnumerator3
    Skeletal Mesh=$User.SkeletalMesh
   ) {module=SkeletalMeshLocation001}
   /SetVariables(Particles.test:NiagaraFloat=$System.TimeOfDayCurve) {enabled=false, isSetParameters=true}
   &SphereLocation(
    MaskSpawnGroup=false
    Override Seed=false
    Sphere Radius=30
    Surface Only=false
    UseNonUniformScale=false
    UseOffset=false
    UseSphereOrientation=false
   ) {enabled=false}
   &BoxLocation(
    Box Origin:Vector3f=$Output.SkeletalMeshLocation001.SampledPosition
    Box Size=<100 100 50>
    MaskSpawnGroup=false
    Module.Override Seed=false
    Rotate Box=false
    Surface Only=false
   )
  ]
  particleUpdateScript [
   &ParticleState()
   &SkeletalMeshLocation(
    Apply To Particle Position=false
    Bone Sampling Mode=::NewEnumerator4
    Inherit Velocity=false
    MaskSpawnGroup=false
    Sampled Position Offset=<-13.1117 -17.3677 2.1285>
    Sampled Position Offset Coordinate Space:ENiagaraMeshTransforms=::NewEnumerator3
    Skeletal Mesh=$User.SkeletalMesh
   ) {enabled=false}
   &self:GetCameraInfo(
    Dot Vector:Vector3f=&MakeCustomVectorFromBool(
     Bool=&SetBoolByFloatComparison(A=&VectorLengthSafe(Vector:Vector3f=$Output.SkeletalMeshLocation.SampledVelocity), B=3)
     True Vector=&SafeNormalizeVector(Vector To Normalize:Vector3f=$Output.SkeletalMeshLocation.SampledVelocity, Fallback Vector:Vector3f=<0 0 0>)
     False Vector=<0 0 0>
    )
    Dot product to return 1=0.8
   ) {enabled=false, module=GetCameraInfo}
   &AddVelocity(
    Coordinate Space=::World
    Scale Added Velocity=&VectorFromFloat(Value=1)
    Velocity=$Particles.Forward Vector World
   ) {enabled=false}
   &AddVelocity(
    Coordinate Space=::World
    Scale Added Velocity=&VectorFromFloat(Value=1)
    Velocity=&MakeCustomVectorFromBool(
     Bool=&SetBoolByFloatComparison(A=&VectorLengthSafe(Vector:Vector3f=$Output.SkeletalMeshLocation.SampledVelocity), B=3)
     True Vector=$Output.SkeletalMeshLocation.SampledVelocity
     False Vector=<0 0 0>
    )
   ) {enabled=false, module=AddVelocity001}
   &SolveForcesAndVelocity(Clamp Velocity=false, Limit Acceleration=false)
   &self:KillParticlesWithLeadup(
    Kill Shape:ENiagaraKillVolumeOptions=::NewEnumerator2
    Length Multiplier=0.015
    Origin Offset:Vector3f=<90 90 0>
    Plane Normal:Vector3f=&RotateVectorByQuaternion(Vector To Rotate=<-1 0 0>, Quaternion=$Engine.Owner.Rotation)
   ) {enabled=false, module=KillParticlesWithLeadup}
   &self:KillParticlesWithLeadup(Invert Volume=true, Length Multiplier=0.05, Origin Offset:Vector3f=<9.1091 5.1925 -5>, Sphere Radius=50) {enabled=false, module=KillParticlesWithLeadup001}
   &Color(Color=&ColorFromCurve(
     ColorCurve=DataInterface() {propertyValues {
       AlphaCurve {keys [<at 0 0.2 RCIM_Cubic>]}
       BlueCurve {keys [<at 0 0 RCIM_Cubic>,<at 0.3333 0>,<at 0.6667 0>,<at 1 1 RCIM_Cubic>]}
       GreenCurve {keys [<at 0 0 RCIM_Cubic>,<at 0.3333 1>,<at 0.6667 1>,<at 1 0 RCIM_Cubic>]}
       RedCurve {keys [<at 0 1 RCIM_Cubic>,<at 0.3333 1>,<at 0.6667 0>,<at 1 0 RCIM_Cubic>]}
      }}
     CurveIndex=$Particles.DistanceFromKillSphere
    )) {enabled=false}
  ]
  rendererClasses [&NiagaraSpriteRendererProperties]
  renderers [{rendererClass=&NiagaraSpriteRendererProperties}]
 }
 ReactiveSmoke {
  emitterUpdateScript [
   &EmitterState()
   &SpawnRate(
    Spawn Probability=&Lerp_Float(A=1, B=0, Alpha=&MakeFloatFromInt(Int=$System.isSkydiving))
    SpawnRate=40
    Use Spawn Probability=true
   )
  ]
  particleSpawnScript [
   &InitializeParticle(
    Color=<rgba 1 1 1 0.2>
    Lifetime=&RandomRangeFloat(Minimum=1, Maximum=2)
    Module.Write Color=true
    Module.Write Lifetime=true
    Module.Write Mass=false
    Module.Write Material Random=true
    Module.Write Position=false
    Module.Write Scale=false
    Module.Write SpriteRotation=true
    Module.Write SpriteSize=true
    Position=/Niagara/Functions/Localspace/SimulationPosition()
    Sprite Rotation=&UniformRangedFloat(Minimum=0, Maximum=360)
    Sprite Size=&Vector2DFromFloat(Value=&UniformRangedFloat(Minimum=4, Maximum=15))
   )
   &AccelerationForce(Acceleration=&UniformRangedVector(Minimum=<0 0 0>, Maximum=<0 0 50>))
   &ApplyInitialForces()
   &SkeletalMeshLocation(
    Bone Sampling Mode=::NewEnumerator4
    Inherit Velocity=false
    MaskSpawnGroup=true
    Sampled Position Offset=&Add_Vector(
     A=<11 5 -3>
     B=&Multiply_Vector(A=<1 1 0.1667>, B=&RandomVector(Vector Scale=6))
    )
    Sampled Position Offset Coordinate Space:ENiagaraMeshTransforms=::NewEnumerator3
    Skeletal Mesh=$User.SkeletalMesh
   )
   &AddVelocity(Velocity=<0 0 50>)
   &InheritVelocity(Module.Limit Speed=false, Velocity Scale=&VectorFromFloat(Value=0.25))
   /SetVariables(Particles.SpawnGroup=&SpawnGroup()) {isSetParameters=true}
  ]
  particleUpdateScript [
   &ParticleState()
   &SkeletalMeshLocation(
    Apply To Particle Position=false
    Bone Sampling Mode=::NewEnumerator4
    Inherit Velocity=false
    MaskSpawnGroup=false
    Sampled Position Offset=&Add_Vector(
     A=<11 5 -3>
     B=&Multiply_Vector(A=<1 1 0.1667>, B=&RandomVector(Vector Scale=6))
    )
    Sampled Position Offset Coordinate Space:ENiagaraMeshTransforms=::NewEnumerator3
    Skeletal Mesh=$User.SkeletalMesh
   ) {module=SkeletalMeshLocation001}
   /SetVariables(Particles.CompareSmokePosition:Vector3f=&Multiply_Vector(
     A=<0.4 0.4 1>
     B=&Subtract_Vector(A=$Output.SkeletalMeshLocation001.SampledPosition, B=Linked() {linkedVariable {name=Particles.Position, type=NiagaraPosition}})
    )) {isSetParameters=true}
   &Drag(Drag=5, Use Linear Drag=true, Use Rotational Drag=false)
   &GravityForce(Gravity=<0 0 100>)
   &InheritVelocity(
    Module.Limit Speed=false
    Velocity Scale=&Multiply_VectorByFloat(
     Vector=<0.1 0.1 0.1>
     Float=&Multiply_Float(
      A=&Lerp_Float(A=1, B=0, Alpha=$Particles.NormalizedAge)
      B=&Lerp_Float(
       A=&RandomRangeFloat(Minimum=2, Maximum=1.1)
       B=0
       Alpha=&Multiply_Float(A=&VectorLengthSafe(Vector:Vector3f=$Particles.CompareSmokePosition), B=0.015)
      )
     )
    )
   ) {module=InheritVelocity001}
   &SolveForcesAndVelocity(Clamp Velocity=false, Limit Acceleration=false)
   &ScaleSpriteSize(Scale Factor=&Vector2DFromFloat(Value=&FloatFromCurve(FloatCurve=DataInterface() {propertyValues {Curve {keys [{arriveTangent=-0.28, interpMode=RCIM_Cubic, leaveTangent=-0.28, tangentMode=RCTM_User, time=0, value=0.5}, <at 1 1>]}}}))) {module=ScaleSpriteSize001}
   &ScaleSpriteSize(Scale Factor=&Vector2DFromFloat(Value=&FloatFromCurve(
      FloatCurve=DataInterface() {propertyValues {Curve {keys [
          {arriveTangent=-0.0248, interpMode=RCIM_Cubic, leaveTangent=-0.0248, tangentMode=RCTM_User, time=0, value=1}
          {arriveTangent=-0.5458, interpMode=RCIM_Cubic, leaveTangent=-0.5458, tangentMode=RCTM_User, time=1, value=0.75}
         ]}}}
      CurveIndex=&ClampFloat(Float=&Multiply_Float(A=&VectorLengthSafe(Vector:Vector3f=$Particles.CompareSmokePosition), B=0.015))
      Scale Curve=2
     )))
   &DynamicMaterialParameters(
    Index 0 Param 1=&FloatFromCurve(FloatCurve=DataInterface() {propertyValues {Curve {keys [
         {arriveTangent=4.0913, interpMode=RCIM_Cubic, leaveTangent=4.0913, tangentMode=RCTM_User, time=0, value=0}
         {arriveTangent=0.1513, interpMode=RCIM_Cubic, leaveTangent=0.1513, tangentMode=RCTM_User, time=0.5, value=1}
         {arriveTangent=-0.0082, interpMode=RCIM_Cubic, leaveTangent=-0.0082, tangentMode=RCTM_User, time=1, value=0}
        ]}}})
    Index 0 Param 4=1
    Module.Param0WriteEnabled=true
    Module.Param1WriteEnabled=false
    Module.Param2WriteEnabled=false
    Module.Param3WriteEnabled=true
   )
   /SetVariables(Particles.StartFrame:NiagaraFloat=&RandomRangeFloat(Minimum=0, Maximum=63)) {isSetParameters=true}
   /SetVariables(Particles.SubImageIndex=&Lerp_Float(
     A=$Particles.StartFrame
     B=&Add_Float(A=$Particles.StartFrame, B=&RandomRangeFloat(Minimum=2, Maximum=6))
     Alpha=$Particles.NormalizedAge
    )) {isSetParameters=true}
   &ScaleColor(
    Scale Alpha=&FloatFromCurve(FloatCurve=DataInterface() {propertyValues {Curve {keys [
         {arriveTangent=-2.2302, interpMode=RCIM_Cubic, leaveTangent=-2.2302, tangentMode=RCTM_User, time=0, value=1}
         {arriveTangent=-0.022, interpMode=RCIM_Cubic, leaveTangent=-0.022, tangentMode=RCTM_User, time=1, value=0}
        ]}}})
    ScaleA=true
    ScaleRGB=false
   )
   &ScaleColor(
    Scale Alpha=&LerpMultipleFloats(
     Float Count:ENiagara_MultipleLerpCount=::NewEnumerator0
     1:NiagaraFloat=&RandomRangeFloat(Minimum=0.02, Maximum=0.04)
     2:NiagaraFloat=&RandomRangeFloat(Minimum=0.05, Maximum=0.1)
     3:NiagaraFloat=&RandomRangeFloat(Minimum=0.02, Maximum=0.04)
     4:NiagaraFloat=?
     5:NiagaraFloat=?
     Alpha:NiagaraFloat=Linked() {linkedVariable {name=NPC.FortniteNPC.TimeOfDay_0-1, type=NiagaraFloat}}
    )
    ScaleA=true
    ScaleRGB=false
   ) {module=ScaleColor002}
   &ScaleColor(
    Scale Alpha=&Lerp_Float(
     A=1
     B=0
     Alpha=&Multiply_Float(A=&VectorLengthSafe(Vector:Vector3f=$Particles.CompareSmokePosition), B=0.015)
    )
    ScaleA=true
    ScaleRGB=false
   ) {module=ScaleColor001}
   &Color(Color=&ColorFromCurve(
     ColorCurve=DataInterface() {propertyValues {
       AlphaCurve {keys [<at 0 1 RCIM_Cubic>]}
       BlueCurve {keys [<at 0 0 RCIM_Cubic>,<at 0.3333 0>,<at 0.6667 0>,<at 1 1 RCIM_Cubic>]}
       GreenCurve {keys [<at 0 0 RCIM_Cubic>,<at 0.3333 1>,<at 0.6667 1>,<at 1 0 RCIM_Cubic>]}
       RedCurve {keys [<at 0 1 RCIM_Cubic>,<at 0.3333 1>,<at 0.6667 0>,<at 1 0 RCIM_Cubic>]}
      }}
     CurveIndex=&Lerp_Float(
      A=1
      B=0
      Alpha=&Multiply_Float(A=&VectorLengthSafe(Vector:Vector3f=$Particles.CompareSmokePosition), B=0.015)
     )
    )) {enabled=false}
   &KillParticles(Kill Particles=&SetBoolByFloatComparison(
     A=&MakeFloatFromLinearColor(LinearColor=$Particles.Color, Channel:ENiagaraLinearColor_Channels=::NewEnumerator3)
     B=0.001
     Comparison Type=::NewEnumerator7
    ))
  ]
  rendererClasses [&NiagaraSpriteRendererProperties]
  renderers [{rendererClass=&NiagaraSpriteRendererProperties}]
 }
 Secondary_Gloops {
  emitterUpdateScript [
   &EmitterState(Enable Distance Culling=true, MaxDistance=7500, Module.bUseMaxDistance=true, Module.bUseMinDistance=false, Scalability Mode=::NewEnumerator1)
   &SpawnPerUnit(
    Module.Use Max Movement Threshold=true
    Module.Use Movement Tolerance=false
    Spawn Per Unit:NiagaraFloat=&FN_FrontEndFloatModifier(Gameplay=1, Front End=1.5)
    Use Spawn Probability=false
    Velocity Vector:Vector3f=&self:GetBoneVelocity(Filtered Bone Index:NiagaraInt32=4, Skeletal Mesh:NiagaraDataInterfaceSkeletalMesh=$User.SkeletalMesh)
   )
  ]
  particleSpawnScript [
   &SkeletalMeshLocation(
    Bone / Socket Index=4
    Bone Sampling Mode=::NewEnumerator4
    Inherit Velocity=false
    MaskSpawnGroup=false
    Sampled Position Offset=<-2 0 -5>
    Sampled Position Offset Coordinate Space:ENiagaraMeshTransforms=::NewEnumerator2
    Skeletal Mesh=$User.SkeletalMesh
   ) {module=SkeletalMeshLocation001}
   &SkeletalMeshLocation(
    Apply To Particle Position=false
    Bone / Socket Index=5
    Bone Sampling Mode=::NewEnumerator4
    Inherit Velocity=false
    MaskSpawnGroup=false
    Sampled Position Offset=&FN_FrontEndVector3Modifier(Gameplay=<-2 0 -5>, Front End=<-2 0 -4>)
    Sampled Position Offset Coordinate Space:ENiagaraMeshTransforms=::NewEnumerator2
    Skeletal Mesh=$User.SkeletalMesh
   ) {module=SkeletalMeshLocation002}
   &InitializeParticle(
    Color=&MakeCustomLinearColorFromBool(
     Bool=$User.bIsFrontEndPreview
     True Color=<rgba 1 0.7398 0.578 3>
     False Color=&RandomRangeLinearColor(Minimum=<rgba 0.1178 0.1273 0.181 3>, Maximum=<rgba 0.232 0.232 0.232 3>)
    )
    Lifetime=&Lerp_Float(
     A=0.4
     B=&RandomRangeFloat(Minimum=0.3, Maximum=0.35)
     Alpha=&ClampFloat(Float=&Add_Float(
       A=&Multiply_Float(A=0.02, B=&VectorLength(Vector=$Output.SkeletalMeshLocation001.SampledVelocity))
       B=-0.2
      ))
    )
    Module.Write Color=true
    Module.Write Lifetime=true
    Module.Write Mass=false
    Module.Write Material Random=true
    Module.Write Position=false
    Module.Write Scale=false
    Module.Write SpriteRotation=false
    Module.Write SpriteSize=true
    Position=&SimulationPosition()
    Sprite Rotation=&UniformRangedFloat(Minimum=0, Maximum=360)
    Sprite Size=&FN_FrontEndVector2Modifier(
     Front End:Vector2f=&RandomRangeVector2D(Minimum=<12 12>, Maximum=<20 20>)
     Gameplay:Vector2f=&RandomRangeVector2D(Minimum=<17 17>, Maximum=<20 20>)
    )
   )
   &SubUVAnimation(Number Of Frames=3, SubUV Animation Mode=::NewEnumerator1)
   &AddVelocityFromPoint(
    Module.Use Falloff Distance=false
    Origin Offset=<0 -12 1>
    Velocity Origin:Vector3f=$Output.SkeletalMeshLocation002.SampledPosition
    Velocity Strength=1
   )
   &AddVelocity(Coordinate Space=::Simulation, Velocity=<0 0 -0.1>)
   /SetVariables(
    Particles.PositionTest:Vector3f=&Add_Vector(A=<0 -12 1>, B=$Output.SkeletalMeshLocation002.SampledPosition)
    Particles.ScaleTest:Vector2f=<5 5>
   ) {isSetParameters=true}
  ]
  particleUpdateScript [
   &self:SetSpriteFacing(New Camera Query=DataInterface() {propertyValues {PlayerControllerIndex=0, bRequireCurrentFrameData=false}}) {module=SetSpriteFacing}
   /SetVariables(
    Particles.SpriteFacing=&Lerp_Vector(
     A=&SafeNormalizeVector(Vector To Normalize:Vector3f=$Particles.Velocity, Fallback Vector:Vector3f=<0 0 0>)
     B=$Particles.SpriteFacingCamera
     Alpha=&FN_FrontEndFloatModifier(Gameplay=1, Front End=0.5)
    )
    Particles.SpriteRotation:NiagaraFloat=&RandomRangeFloat(Minimum=-75, Maximum=75)
   ) {isSetParameters=true}
   &ParticleState()
   &SolveForcesAndVelocity(Clamp Velocity=false, Limit Acceleration=false)
   &ScaleColor(
    Scale Alpha=&FloatFromCurve(FloatCurve=DataInterface() {propertyValues {Curve {keys [<at 0.0481 1>,<at 0.179 0.5532>,<at 0.8619 0.544>,<at 1 0>]}}})
    Scale RGB=&VectorFromCurve(
     VectorCurve=DataInterface() {propertyValues {XCurve {keys [<at 0 1.5>,<at 1 1>]}, YCurve {keys [<at 0 1.5>,<at 1 1>]}, ZCurve {keys [<at 0 1.5>,<at 1 1>]}}}
     Scale Curve=&VectorFromFloat(Value=&LerpMultipleFloats(
       Float Count:ENiagara_MultipleLerpCount=::NewEnumerator0
       1:NiagaraFloat=0.3
       2:NiagaraFloat=1
       3:NiagaraFloat=0.3
       4:NiagaraFloat=?
       5:NiagaraFloat=?
       Alpha:NiagaraFloat=Linked() {linkedVariable {name=NPC.FortniteNPC.TimeOfDay_0-1, type=NiagaraFloat}}
      ))
    )
    ScaleA=true
    ScaleRGB=true
   )
   &DynamicMaterialParameters(
    Index 0 Param 1=&Lerp_Float(A=1, B=0, Alpha=$Particles.NormalizedAge)
    Index 0 Param 2=&FN_FrontEndFloatModifier(
     Gameplay=0
     Front End=&FloatFromCurve(FloatCurve=DataInterface() {propertyValues {Curve {keys [
          {arriveTangent=0.2251, interpMode=RCIM_Cubic, leaveTangent=0.2251, tangentMode=RCTM_User, time=0, value=0}
          {arriveTangent=3.1399, interpMode=RCIM_Cubic, leaveTangent=3.1399, tangentMode=RCTM_User, time=1, value=1}
         ]}}})
    )
    Module.Param0WriteEnabled=true
    Module.Param1WriteEnabled=true
    Module.Param2WriteEnabled=false
    Module.Param3WriteEnabled=false
   )
   &ScaleSpriteSize(Scale Factor=&Vector2DFromFloat(Value=&FloatFromCurve(FloatCurve=DataInterface() {propertyValues {Curve {keys [{arriveTangent=2.0298, interpMode=RCIM_Cubic, leaveTangent=2.0298, tangentMode=RCTM_User, time=0, value=0}, <at 1 1>]}}}))) {module=ScaleSpriteSize001}
   &SkeletalMeshLocation(
    Apply To Particle Position=false
    Bone / Socket Index=1
    Bone Sampling Mode=::NewEnumerator4
    Inherit Velocity=false
    MaskSpawnGroup=false
    Sampled Position Offset=<-2 0 -5>
    Sampled Position Offset Coordinate Space:ENiagaraMeshTransforms=::NewEnumerator2
    Skeletal Mesh=$User.SkeletalMesh
   )
   &KillParticles(Kill Particles=&Boolean_AndOperation(
     Bool1=&Boolean_NotOperation(Bool=$User.bIsFrontEndPreview)
     Bool2=&SetBoolByFloatComparison(
      A=&MakeFloatFromVector(
       Vector=&TransformVector(Vector=Linked() {linkedVariable {name=Particles.Position, type=NiagaraPosition}}, Source Space=::Simulation, Destination Space=::World)
       Channel=::NewEnumerator2
      )
      B=&Add_Float(
       A=-2
       B=&MakeFloatFromVector(Vector=/Niagara/DynamicInputs/Helpers/SimulationPosition(), Channel=::NewEnumerator2)
      )
     )
    )) {module=KillParticles001}
   &KillParticlesInVolume(
    Box Size:Vector3f=<100 100 54>
    Kill Shape=::NewEnumerator1
    Kill Volume Enabled:NiagaraBool=$User.bIsFrontEndPreview
    Volume Origin:Vector3f=$Output.SkeletalMeshLocation.SampledPosition
   )
   &self:KillParticlesWithLeadup(Invert Volume=true, Length Multiplier=0.05, Origin Offset:Vector3f=<9.1091 5.1925 -5>, Sphere Radius=50) {enabled=false, module=KillParticlesWithLeadup}
   &CameraOffset(Camera Offset Amount=&Lerp_Float(
     A=0
     B=&Lerp_Float(
      A=10
      B=40
      Alpha=&ClampFloat(Float=&Add_Float(
        A=&Multiply_Float(A=0.02, B=&VectorLength(Vector=$Output.SkeletalMeshLocation.SampledVelocity))
        B=-0.2
       ))
     )
     Alpha=$Particles.NormalizedAge
    )) {enabled=false}
  ]
  rendererClasses [&NiagaraSpriteRendererProperties]
  renderers [
   {rendererClass=&NiagaraSpriteRendererProperties}
   {rendererClass=&NiagaraSpriteRendererProperties, rendererIndex=1}
  ]
 }
 SmokeRibbon {
  emitterUpdateScript [
   &EmitterState()
   &SpawnPerUnit(
    Module.Use Max Movement Threshold=true
    Module.Use Movement Tolerance=false
    Spawn Per Unit=15
    Spawn Probability:NiagaraFloat=&Lerp_Float(A=1, B=0, Alpha=&MakeFloatFromInt(Int=$System.isSkydiving))
    Use Spawn Probability=true
   )
   &SpawnRate(SpawnRate=20, Use Spawn Probability=false) {enabled=false}
  ]
  particleSpawnScript [
   /SetVariables(Particles.Lifetime=0.5, Particles.RibbonWidth=50) {isSetParameters=true}
   &SkeletalMeshLocation(
    Bone / Socket Index=2
    Bone Sampling Mode=::NewEnumerator4
    Inherit Velocity=false
    MaskSpawnGroup=false
    Sampled Position Offset=<-13.1117 -11 2.1285>
    Sampled Position Offset Coordinate Space:ENiagaraMeshTransforms=::NewEnumerator3
    Skeletal Mesh=$User.SkeletalMesh
   )
   &AddVelocity(Velocity=<0 -1000 0>) {enabled=false, module=AddVelocity001}
   /SetVariables(Particles.RibbonLinkOrder:NiagaraFloat=&MakeFloatFromInt(Int=/Niagara/DynamicInputs/DataInterface/GenerateIntFromCounter(SimpleCounter:NiagaraDataInterfaceSimpleCounter=DataInterface() {propertyValues {GpuSyncMode=None, InitialValue=0}}))) {isSetParameters=true}
   &InheritVelocity(Module.Limit Speed=false, Velocity Scale=&VectorFromFloat(Value=1)) {enabled=false}
   &SkeletalMeshLocation(
    Apply To Particle Position=false
    Bone Sampling Mode=::NewEnumerator4
    Inherit Velocity=false
    MaskSpawnGroup=false
    Sampled Position Offset=<-13.1117 -11 2.1285>
    Sampled Position Offset Coordinate Space:ENiagaraMeshTransforms=::NewEnumerator3
    Skeletal Mesh=$User.SkeletalMesh
   ) {module=SkeletalMeshLocation001}
   /SetVariables(
    Particles.MeshOrientationTest:Quat4f=$Output.SkeletalMeshLocation001.SampledRotation
    Particles.RibbonFacing=&RotateVectorByQuaternion(Vector To Rotate=<-1 0 0>, Quaternion=$Output.SkeletalMeshLocation001.SampledRotation)
    Particles.RibbonFacingAlt:Vector3f=&RotateVectorByQuaternion(Vector To Rotate=<1 0 0>, Quaternion=$Output.SkeletalMeshLocation001.SampledRotation)
    Particles.RibbonFacingAltAlt:Vector3f=&RotateVectorByQuaternion(Vector To Rotate=<0 0 -1>, Quaternion=$Output.SkeletalMeshLocation001.SampledRotation)
   ) {isSetParameters=true}
  ]
  particleUpdateScript [
   &ParticleState()
   &SkeletalMeshLocation(
    Apply To Particle Position:NiagaraBool=&SetBoolByFloatComparison(A=$Particles.NormalizedAge, B=0.25, Comparison Type=::NewEnumerator6)
    Bone Sampling Mode=::NewEnumerator4
    Inherit Velocity=false
    MaskSpawnGroup=false
    Sampled Position Offset Coordinate Space:ENiagaraMeshTransforms=::NewEnumerator3
    Skeletal Mesh=$User.SkeletalMesh
   ) {module=SkeletalMeshLocation003}
   &Drag(Drag=5, Use Linear Drag=true, Use Rotational Drag=false)
   &GravityForce(Gravity=<0 0 600>)
   &SolveForcesAndVelocity(Clamp Velocity=false, Limit Acceleration=false)
   /Game/Effects/Niagara/Modules/Materials/FN_RibbonMaterial/FN_RibbonMaterial_v2(
    Color:LinearColor=&MakeLinearColorFromVectorAndFloat(
     "Vector (RGB)"=<1 1 1>
     "Float (Alpha)"=&FloatFromCurve(
      FloatCurve=DataInterface() {propertyValues {Curve {keys [
          {arriveTangent=0.0082, interpMode=RCIM_Cubic, leaveTangent=0.0082, tangentMode=RCTM_User, time=0.25, value=0}
          {arriveTangent=-0.0105, interpMode=RCIM_Cubic, leaveTangent=-0.0105, tangentMode=RCTM_User, time=0.4453, value=0.1}
          {arriveTangent=-0.4528, interpMode=RCIM_Cubic, leaveTangent=-0.4528, tangentMode=RCTM_User, time=1, value=0}
         ]}}}
      Scale Curve=&LerpMultipleFloats(
       Float Count:ENiagara_MultipleLerpCount=::NewEnumerator0
       1:NiagaraFloat=0.65
       2:NiagaraFloat=1.05
       3:NiagaraFloat=0.65
       4:NiagaraFloat=?
       5:NiagaraFloat=?
       Alpha:NiagaraFloat=Linked() {linkedVariable {name=NPC.FortniteNPC.TimeOfDay_0-1, type=NiagaraFloat}}
      )
     )
    )
    Color Pan U Speed=2.8
    Color Scale U=2
    Erosion Amount:NiagaraFloat=&Lerp_Float(A=0, B=1.75, Alpha=$Particles.NormalizedAge)
    Erosion Pan Speed U=2.75
    Erosion Power=0.25
    Erossion Scale U=2
    Opacity Pan U Speed=0.23
    Opacity Scale U=0.2
    Scale Alpha:NiagaraFloat=&FloatFromCurve(
     FloatCurve=DataInterface() {propertyValues {Curve {keys [
         {interpMode=RCIM_Cubic, tangentMode=RCTM_User, time=0, value=1}
         {arriveTangent=-2.8707, interpMode=RCIM_Cubic, leaveTangent=-2.8707, tangentMode=RCTM_User, time=1, value=0}
        ]}}}
     Scale Curve=2
    )
    TIme:NiagaraFloat=$Engine.Time
   )
   &ScaleRibbonWidth(Ribbon Width Scale=&FloatFromCurve(FloatCurve=DataInterface() {propertyValues {Curve {keys [
         {arriveTangent=0.4163, interpMode=RCIM_Cubic, leaveTangent=0.4163, tangentMode=RCTM_User, time=0.3, value=1}
         {arriveTangent=-1.9468, interpMode=RCIM_Cubic, leaveTangent=-1.9468, tangentMode=RCTM_User, time=1.0032, value=0.0313}
        ]}}}))
   &self:GetCameraInfo(
    Dot Vector:Vector3f=&MakeCustomVectorFromBool(
     Bool=&SetBoolByFloatComparison(A=&VectorLengthSafe(Vector:Vector3f=$Output.SkeletalMeshLocation003.SampledVelocity), B=3)
     True Vector=&SafeNormalizeVector(Vector To Normalize:Vector3f=$Output.SkeletalMeshLocation003.SampledVelocity, Fallback Vector:Vector3f=<0 0 0>)
     False Vector=<0 0 0>
    )
    Dot product to return 1=0.95
    Interpolate Value=5
   ) {module=GetCameraInfo}
   &ScaleColor(Color Value To Scale=$Particles.Color, Scale Alpha=$Particles.CameraVelocityDot, ScaleA=true, ScaleRGB=false)
   &Color(Color=&ColorFromCurve(
     ColorCurve=DataInterface() {propertyValues {
       AlphaCurve {keys [<at 0 1 RCIM_Cubic>]}
       BlueCurve {keys [<at 0 0 RCIM_Cubic>,<at 0.3333 0>,<at 0.6667 0>,<at 1 1 RCIM_Cubic>]}
       GreenCurve {keys [<at 0 0 RCIM_Cubic>,<at 0.3333 1>,<at 0.6667 1>,<at 1 0 RCIM_Cubic>]}
       RedCurve {keys [<at 0 1 RCIM_Cubic>,<at 0.3333 1>,<at 0.6667 0>,<at 1 0 RCIM_Cubic>]}
      }}
     CurveIndex=$Particles.CameraVelocityDot
    )) {enabled=false}
  ]
  rendererClasses [&NiagaraRibbonRendererProperties, &NiagaraMeshRendererProperties]
  renderers [
   {rendererClass=&NiagaraRibbonRendererProperties}
   {rendererClass=&NiagaraMeshRendererProperties, rendererIndex=1}
   {rendererClass=&NiagaraRibbonRendererProperties, rendererIndex=2}
   {rendererClass=&NiagaraRibbonRendererProperties, rendererIndex=3}
  ]
 }
}
summary {
 systemName=NS_PieMan_Backpack
 emitters [
  {emitterName=DroolSource, rendererClasses [&NiagaraSpriteRendererProperties]}
  {emitterName=Drool, rendererClasses [&NiagaraMeshRendererProperties, &NiagaraSpriteRendererProperties]}
  {emitterName=SmokeRibbon, rendererClasses [&NiagaraRibbonRendererProperties, &NiagaraMeshRendererProperties]}
  {emitterName=Flies_Optimized, rendererClasses [&NiagaraMeshRendererProperties]}
  {emitterName=Fly_Trail, rendererClasses [&NiagaraRibbonRendererProperties]}
  {emitterName=ReactiveSmoke, rendererClasses [&NiagaraSpriteRendererProperties]}
  {bEnabled=false, emitterName=NE_Empty, rendererClasses [&NiagaraSpriteRendererProperties]}
  {emitterName=Main_Gloop, rendererClasses [&NiagaraSpriteRendererProperties]}
  {emitterName=Secondary_Gloops, rendererClasses [&NiagaraSpriteRendererProperties]}
  {emitterName=DroolPendulum, rendererClasses [&NiagaraSpriteRendererProperties, &NiagaraMeshRendererProperties]}
  {emitterName=Drool001, rendererClasses [&NiagaraSpriteRendererProperties, &NiagaraMeshRendererProperties]}
 ]
 userVariables [
  User.bIsFrontEnd {type=NiagaraBool}
  User.bIsFrontEndPreview {type=NiagaraBool}
  User.ForwardDot {type=NiagaraFloat, defaultValue=0, description=""}
  User.IsSkydiving {type=NiagaraBool}
  User.JumpTime {type=NiagaraFloat, defaultValue=0, description=""}
  User.SkeletalMesh {
   type=NiagaraDataInterfaceSkeletalMesh
   defaultValue {
    struct=NiagaraExt_VariableValue_DataInterface
    value {dataInterface=&self:NiagaraDataInterfaceSkeletalMesh_0}
   }
  }
  User.SpawnLandedFX {type=NiagaraBool}
 ]
}
