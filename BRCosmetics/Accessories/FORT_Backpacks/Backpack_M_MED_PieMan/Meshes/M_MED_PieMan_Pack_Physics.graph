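# Physics asset for the PieMan backpack cosmetic, giving its cauldron goo and overhanging goop bones bodies for secondary motion.
# A large capsule on the pack bone anchors two constrained groups: a two-capsule standing-goo chain and an overhanging-goop box that branches into left and right two-box chains.
# Angular limits keep the standing goo stiffest, allow moderate swing at the overhang root and the widest along its branches, with tight twist throughout.
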PhysicsAsset /BRCosmetics/Accessories/FORT_Backpacks/Backpack_M_MED_PieMan/Meshes/M_MED_PieMan_Pack_Physics
bodies {
 DYN_CauldronStandingGoo_01 [{shapeType=Capsule, length=0.3438, radius=0.5, center=<0.63 0 -0.0001>, rotation=<rot 90 0 0>}]
 DYN_CauldronStandingGoo_02 [{shapeType=Capsule, length=0.9062, radius=0.7188, center=<1.2491 0 0>, rotation=<rot 90 0 0>}]
 DYN_OverhangingGoop [{shapeType=Box, extentX=1.25, extentY=1, extentZ=2, center=<0.7272 0 0>, rotation=<rot 0 0 -36.5638>}]
 DYN_OverhangingGoopL_01 [{shapeType=Box, extentX=1.8125, extentY=1, extentZ=1, center=<0.9338 0 0>}]
 DYN_OverhangingGoopL_02 [{shapeType=Box, extentX=2.3125, extentY=1, extentZ=1, center=<1.2321 0 0>}]
 DYN_OverhangingGoopR_01 [{shapeType=Box, extentX=1.6875, extentY=1, extentZ=1, center=<0.8853 0 0>}]
 DYN_OverhangingGoopR_02 [{shapeType=Box, extentX=1.875, extentY=1, extentZ=1, center=<0.9797 -0.0001 0>}]
 pack [{shapeType=Capsule, length=4.9062, radius=10.7536, center=<-0.8528 -13.9131 -9.9581>, rotation=<rot -5.5197 113.1286 -151.0173>}]
}
constraints [
 {bone1Name=DYN_OverhangingGoopR_01, bone2Name=DYN_OverhangingGoop, swing1LimitDegrees=45, swing2LimitDegrees=45, twistLimitDegrees=5}
 {bone1Name=DYN_OverhangingGoopR_02, bone2Name=DYN_OverhangingGoopR_01, swing1LimitDegrees=45, swing2LimitDegrees=45, twistLimitDegrees=5}
 {bone1Name=DYN_OverhangingGoopL_01, bone2Name=DYN_OverhangingGoop, swing1LimitDegrees=45, swing2LimitDegrees=45, twistLimitDegrees=5}
 {bone1Name=DYN_OverhangingGoopL_02, bone2Name=DYN_OverhangingGoopL_01, swing1LimitDegrees=45, swing2LimitDegrees=45, twistLimitDegrees=5}
 {
  bone1Name=DYN_CauldronStandingGoo_02
  bone2Name=DYN_CauldronStandingGoo_01
  swing1LimitDegrees=13
  swing2LimitDegrees=20
  twistLimitDegrees=5
 }
 {bone1Name=DYN_OverhangingGoop, bone2Name=pack, swing1LimitDegrees=25, swing2LimitDegrees=25, twistLimitDegrees=5}
 {bone1Name=DYN_CauldronStandingGoo_01, bone2Name=pack, swing1LimitDegrees=10, swing2LimitDegrees=20, twistLimitDegrees=5}
]
