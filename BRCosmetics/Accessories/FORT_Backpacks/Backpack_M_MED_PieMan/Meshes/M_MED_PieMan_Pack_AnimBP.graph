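# Animation Blueprint for the PieMan backpack cosmetic, playing an idle with physics-driven secondary motion.
# The Pieman_Idle sequence gets an additive offset and yaw on the pack bone, then feeds a RigidBody node using the pack's physics asset in root-bone space.
# A chain of four spring-based AnimDynamics nodes for the overhanging goop bones is not connected to the output pose.
# The event graph holds no logic.
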AnimBlueprint /BRCosmetics/Accessories/FORT_Backpacks/Backpack_M_MED_PieMan/Meshes/M_MED_PieMan_Pack_AnimBP
target_skeleton=&/M_MED_PieMan_Pack_Skeleton
graphs [
 AnimGraph {nodes [
   n1 = Anim Dynamics - Bone: DYN_OverhangingGoopR_01(Alpha=1, ExternalForce="0,0,0") {
    bGravityOverrideInSimSpace=true
    bLinearSpring=true
    gravityScale=0.2
    linearSpringConstant=10
    simulationSpace=BoneRelative
    alphaBoolBlend {blendOption=Linear}
    boundBone {boneName=DYN_OverhangingGoopR_01}
    componentLinearAccScale=<0.3 0.3 3>
    componentLinearVelScale=<0.6 0.6 0.6>
    physicsBodyDefinitions [{
      sphereCollisionRadius=0
      boundBone {boneName=DYN_OverhangingGoopR_01}
      boxExtents=<20 20 20>
      constraintSetup {angularLimitsMax=<0 0 0>, angularLimitsMin=<0 0 0>, angularTarget=<0 0 0>, linearAxesMax=<0.5 0 0>, linearAxesMin=<-0.5 0 0>}
      localJointOffset=<0 -5 0>
     }]
    relativeSpaceBone {boneName=root}
   }
   n2 = Anim Dynamics - Bone: DYN_OverhangingGoopR_02(ComponentPose=@n1, Alpha=1, ExternalForce="0,0,0") {
    bGravityOverrideInSimSpace=true
    bLinearSpring=true
    gravityScale=0.2
    linearSpringConstant=10
    simulationSpace=BoneRelative
    alphaBoolBlend {blendOption=Linear}
    boundBone {boneName=DYN_OverhangingGoopR_02}
    componentLinearAccScale=<0.3 0.3 3>
    componentLinearVelScale=<0.6 0.6 0.6>
    physicsBodyDefinitions [{
      sphereCollisionRadius=0
      boundBone {boneName=DYN_OverhangingGoopR_02}
      boxExtents=<20 20 20>
      constraintSetup {angularLimitsMax=<0 0 0>, angularLimitsMin=<0 0 0>, angularTarget=<0 0 0>, linearAxesMax=<0.5 0 0>, linearAxesMin=<-0.5 0 0>}
      localJointOffset=<0 -5 0>
     }]
    relativeSpaceBone {boneName=root}
   }
   n3 = Anim Dynamics - Bone: DYN_OverhangingGoopL_01(ComponentPose=@n2, Alpha=1, ExternalForce="0,0,0") {
    bGravityOverrideInSimSpace=true
    bLinearSpring=true
    gravityScale=0.2
    linearSpringConstant=10
    simulationSpace=BoneRelative
    alphaBoolBlend {blendOption=Linear}
    boundBone {boneName=DYN_OverhangingGoopL_01}
    componentLinearAccScale=<0.3 0.3 3>
    componentLinearVelScale=<0.6 0.6 0.6>
    physicsBodyDefinitions [{
      sphereCollisionRadius=0
      boundBone {boneName=DYN_OverhangingGoopL_01}
      boxExtents=<20 20 20>
      constraintSetup {angularLimitsMax=<0 0 0>, angularLimitsMin=<0 0 0>, angularTarget=<0 0 0>, linearAxesMax=<0.5 0 0>, linearAxesMin=<-0.5 0 0>}
      localJointOffset=<0 -5 0>
     }]
    relativeSpaceBone {boneName=root}
   }
   Anim Dynamics - Bone: DYN_OverhangingGoopL_02(ComponentPose=@n3, Alpha=1, ExternalForce="0,0,0") {
    bGravityOverrideInSimSpace=true
    bLinearSpring=true
    gravityScale=0.2
    linearSpringConstant=10
    simulationSpace=BoneRelative
    alphaBoolBlend {blendOption=Linear}
    boundBone {boneName=DYN_OverhangingGoopL_02}
    componentLinearAccScale=<0.3 0.3 3>
    componentLinearVelScale=<0.6 0.6 0.6>
    physicsBodyDefinitions [{
      sphereCollisionRadius=0
      boundBone {boneName=DYN_OverhangingGoopL_02}
      boxExtents=<20 20 20>
      constraintSetup {angularLimitsMax=<0 0 0>, angularLimitsMin=<0 0 0>, angularTarget=<0 0 0>, linearAxesMax=<0.5 0 0>, linearAxesMin=<-0.5 0 0>}
      localJointOffset=<0 -5 0>
     }]
    relativeSpaceBone {boneName=root}
   }
   Local Space Ref Pose()
   n4 = "Transform (Modify) Bone - Bone: pack"(
    ComponentPose=Local To Component(LocalPose=Sequence Player 'Pieman_Idle'() {sequence=/BRCosmetics/Accessories/FORT_Backpacks/Backpack_M_MED_PieMan/Animation/Pieman_Idle})
    Translation="3,-2,0"
    Rotation="0,100,0"
   ) {
    rotationMode=BMM_Additive
    translationMode=BMM_Additive
    alphaBoolBlend {blendOption=Linear}
    boneToModify {boneName=pack}
    rotation {yaw=100}
    translation=<3 -2>
   }
   n5 = RigidBody(ComponentPose=@n4) {
    bOverrideWorldGravity=true
    overridePhysicsAsset=&/M_MED_PieMan_Pack_Physics
    simulationSpace=BaseBoneSpace
    alphaBoolBlend {blendOption=Linear}
    baseBoneRef {boneName=root}
    componentLinearAccScale=<0.3 0.3 0.3>
    componentLinearVelScale=<0.6 0.6 0.6>
    overrideWorldGravity {z=-980}
   }
   Output Pose(Result=Component To Local(ComponentPose=@n5)) {layerGroup=None}
  ]}
 EventGraph {nodes [
   TryGetPawnOwner()
   Event BlueprintUpdateAnimation()
  ]}
]
variables []
